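; Global foliage control actor that drives wind, health and color variation for foliage materials through the QMPC_GlobalFoliageActor material parameter collection.
; RunGlobalFoliageActor runs on construction, BeginPlay and every Tick, pushing separate tree and plant wind settings into the collection.
; Wind Direction is derived from the root icon's yaw, normalized to a turn fraction with an offset chosen by Flip Wind Direction.
; It also orients the wind-sock, base and arrow icons.
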(event EventTick
  (params (DeltaSeconds :type Float%20%28single-precision%29))
  (Wind|RunGlobalFoliageActor))

(event EventBeginPlay
  (Wind|RunGlobalFoliageActor))

(fn RunGlobalFoliageActor
  (bind _icon_root (Variables|Default|GetIcon_Root))
  (bind _returnvalue (Transformation|GetWorldRotation _icon_root))
  (bind _yaw (.yaw _returnvalue))
  (bind _icon_sock (Variables|Default|GetIcon_Sock))
  (bind _flip_wind_direction (Variables|Wind|GetFlip%20Wind%20Direction))
  ;[1 1 1] Wind
  (if _flip_wind_direction
    ;[1 1 1] Flipped Icon Function
    (Transformation|SetRelativeRotation _icon_sock "0,180.000000, 0" false false)
    (Transformation|SetRelativeLocation _icon_sock "-220.000000, 0, 0" false false) ; fix location of wind sock center to not intersect with the trees
  )
  (if _flip_wind_direction
    (if (< _yaw 0.0)
      (Rendering|Material|SetScalarParameterValue "/Fab/MaterialParameterCollection/QMPC_GlobalFoliageActor.QMPC_GlobalFoliageActor" "Wind Direction" (+ (* (/ (+ _yaw 3.6e+02) 3.6e+02) -1.0) 1.625) :!class /Script/Engine.KismetMaterialLibrary)
      (else
        (Rendering|Material|SetScalarParameterValue "/Fab/MaterialParameterCollection/QMPC_GlobalFoliageActor.QMPC_GlobalFoliageActor" "Wind Direction" (+ (* (/ _yaw 3.6e+02) -1.0) 1.625) :!class /Script/Engine.KismetMaterialLibrary)))
    (elif (< _yaw 0.0)
      (Rendering|Material|SetScalarParameterValue "/Fab/MaterialParameterCollection/QMPC_GlobalFoliageActor.QMPC_GlobalFoliageActor" "Wind Direction" (+ (* (/ (+ _yaw 3.6e+02) 3.6e+02) -1.0) 1.125) :!class /Script/Engine.KismetMaterialLibrary)
      (else
        (Rendering|Material|SetScalarParameterValue "/Fab/MaterialParameterCollection/QMPC_GlobalFoliageActor.QMPC_GlobalFoliageActor" "Wind Direction" (+ (* (/ _yaw 3.6e+02) -1.0) 1.125) :!class /Script/Engine.KismetMaterialLibrary))))
  ;[1 1 1] Global Foliage Actor
  (Transformation|SetWorldRotation (Variables|Default|GetIcon_Base) "0, 0, 0" false false)
  (Transformation|SetWorldRotation (Variables|Default|GetIcon_Arrow_Test) (Math|Rotator|MakeRotator 0.0 0.0 (.yaw (Transformation|GetWorldRotation (Variables|Default|GetIcon_Root)))) false false)
  ;[1 1 1] Wind Parameters - Trees
  (Rendering|Material|SetScalarParameterValue "/Fab/MaterialParameterCollection/QMPC_GlobalFoliageActor.QMPC_GlobalFoliageActor" "Wind Strength Trees" (Variables|Wind|GetWind%20Strength%20Trees) :!class /Script/Engine.KismetMaterialLibrary)
  (Rendering|Material|SetScalarParameterValue "/Fab/MaterialParameterCollection/QMPC_GlobalFoliageActor.QMPC_GlobalFoliageActor" "Wind Speed Trees" (Variables|Wind|GetWind%20Speed%20Trees) :!class /Script/Engine.KismetMaterialLibrary)
  (Rendering|Material|SetScalarParameterValue "/Fab/MaterialParameterCollection/QMPC_GlobalFoliageActor.QMPC_GlobalFoliageActor" "Wind Tiling Trees" (Variables|Wind|GetWind%20Tiling%20Trees) :!class /Script/Engine.KismetMaterialLibrary)
  (Rendering|Material|SetScalarParameterValue "/Fab/MaterialParameterCollection/QMPC_GlobalFoliageActor.QMPC_GlobalFoliageActor" "Wind Noise Trees" (Variables|Wind|GetWind%20Noise%20Trees) :!class /Script/Engine.KismetMaterialLibrary)
  ;[1 1 1] Wind Parameters - Plants
  (Rendering|Material|SetScalarParameterValue "/Fab/MaterialParameterCollection/QMPC_GlobalFoliageActor.QMPC_GlobalFoliageActor" "Wind Speed Plants" (Variables|Wind|GetWind%20Speed%20Plants) :!class /Script/Engine.KismetMaterialLibrary)
  (Rendering|Material|SetScalarParameterValue "/Fab/MaterialParameterCollection/QMPC_GlobalFoliageActor.QMPC_GlobalFoliageActor" "Wind Tiling Plants" (Variables|Wind|GetWind%20Tiling%20Plants) :!class /Script/Engine.KismetMaterialLibrary)
  (Rendering|Material|SetScalarParameterValue "/Fab/MaterialParameterCollection/QMPC_GlobalFoliageActor.QMPC_GlobalFoliageActor" "Wind Strength Plants" (Variables|Wind|GetWind%20Strength%20Plants) :!class /Script/Engine.KismetMaterialLibrary)
  (Rendering|Material|SetScalarParameterValue "/Fab/MaterialParameterCollection/QMPC_GlobalFoliageActor.QMPC_GlobalFoliageActor" "Wind Noise Plants" (Variables|Wind|GetWind%20Noise%20Plants) :!class /Script/Engine.KismetMaterialLibrary)
  ;[1 1 1] Season & Health Parameter
  (Rendering|Material|SetScalarParameterValue "/Fab/MaterialParameterCollection/QMPC_GlobalFoliageActor.QMPC_GlobalFoliageActor" "Health" (Variables|Season&Health|GetHealth) :!class /Script/Engine.KismetMaterialLibrary)
  ;[1 1 1] Color Variation
  (Rendering|Material|SetScalarParameterValue "/Fab/MaterialParameterCollection/QMPC_GlobalFoliageActor.QMPC_GlobalFoliageActor" "Random Color Variation Amount" (Variables|ColorVariation|GetRandom%20Color%20Variation%20Amount) :!class /Script/Engine.KismetMaterialLibrary))

(fn ConstructionScript
  (Wind|RunGlobalFoliageActor))
