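# Two-sided Default Lit material in the Global Foliage Actor content, named as a UI element, built around a wind sway calculation.
# A scaled-down Wind Strength Trees value from the QMPC_GlobalFoliageActor parameter collection drives both intensity and speed of SimpleGrassWind.
# Vertex color alpha weights the sway per vertex.
# The graph has no color or texture logic and shows no link from the wind function to a material output.
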Material /Fab/Actors/GlobalFoliageActor/Materials/MA_UI_Element
graph {
 properties {pixelDepthOffsetMode=PDOM_Legacy, twoSided=true, usedShadingModels=Default Lit}
 nodes [
  Constant()
  n1 = Multiply(A=CollectionParameter() {collection=/Fab/MaterialParameterCollection/QMPC_GlobalFoliageActor, parameterName=Wind Strength Trees}) {constB=0.05}
  MaterialFunctionCall!(WindIntensity=@n1, WindWeight=VertexColor().A, WindSpeed=@n1, AdditionalWPO=Constant3Vector()) {materialFunction=/Engine/Functions/Engine_MaterialFunctions01/WorldPositionOffset/SimpleGrassWind}
  VertexColor()
 ]
}
